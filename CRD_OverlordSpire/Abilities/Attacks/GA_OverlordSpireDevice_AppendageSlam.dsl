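; Gameplay ability for the Overlord Spire device's appendage slam attack, extending a parent slam ability.
; It notifies the spire of attack and recovery, ends when the spire is disabled, and takes activation checks, cooldown and randomized prevent time from it.
; Instant damage uses per-target-type SetByCaller magnitudes, hitting Deimos NPCs only when their damage is positive.
; The spawned ground wave receives this ability and scalable duration and radius values.
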(event Ability|EventActivateAbility
  (|CallParentFunction)
  (bind _asdevice_overlord_spire (Utilities|Casting|CastToDevice_OverlordSpire (Variables|Internal|GetSpireActor)))
  (Abilities|Slam|NotifySlamAttackEvent _asdevice_overlord_spire true)
  (Default|BindEventtoOnDisabled _asdevice_overlord_spire (EventDispatchers|CreateEvent :!event Custom|OnSpireDisabled)))

(event Ability|EventOnEndAbility
  (params (bWasCancelled :type Boolean))
  (|CallParentFunction bWasCancelled)
  (bind _asdevice_overlord_spire (Utilities|Casting|CastToDevice_OverlordSpire (Variables|Internal|GetSpireActor)))
  (Abilities|Slam|NotifySlamAttackEvent _asdevice_overlord_spire false)
  (Default|UnbindEventfromOnDisabled _asdevice_overlord_spire (EventDispatchers|CreateEvent :!event Custom|OnSpireDisabled)))

(event Custom|OnSpireDisabled
  (Ability|EndAbility))

(event EventStartSlamRecovery
  (bind _asdevice_overlord_spire (Utilities|Casting|CastToDevice_OverlordSpire (Variables|Internal|GetSpireActor)))
  (Abilities|Slam|NotifySlamRecoverEvent _asdevice_overlord_spire true)
  (|CallParentFunction))

(event EventPostSlamRecovery
  (bind _asdevice_overlord_spire (Utilities|Casting|CastToDevice_OverlordSpire (Variables|Internal|GetSpireActor)))
  (Abilities|Slam|NotifySlamRecoverEvent _asdevice_overlord_spire false)
  (|CallParentFunction))

(fn ApplySlamDamage
  (params (TargetData :type Gameplay%20Ability%20Target%20Data%20Handle%20Structure))
  ; Apply damage overrides for each possible target type(Player, Building, PBW, Vehicles, Guards, Wildlife, Creatures (Deimos))
  (LocalVariables|SetCurrentEffectSpec (Ability|MakeOutgoingGameplayEffectSpec "/CRD_OverlordSpire/Abilities/Attacks/GE_OverlordSpireDevice_AppendageSlam_InstantDamage.GE_OverlordSpireDevice_AppendageSlam_InstantDamage_C"))
  (bind (_keypin _valuepin) (Utilities|Map|ForEachLoop(Map) (Variables|Default|GetDamageSFValuesPerType))
    (:then
      (bind _returnvalue (Ability|GameplayEffect|AssignTagSetbyCallerMagnitude (LocalVariables|GetCurrentEffectSpec) _keypin (ScalableFloat|GetValueatLevelwithContextParam (Ability|GetAvatarActorfromActorInfo) _valuepin)))
      (LocalVariables|SetCurrentEffectSpec _returnvalue))
    (:CompletedPin
      (Ability|ApplyGameplayEffectSpecToTarget (LocalVariables|GetCurrentEffectSpec) (Variables|GetTargetData)))))

(fn GetAbilityCooldownTime
  (bind _asdevice_overlord_spire (Utilities|Casting|CastToDevice_OverlordSpire (Ability|GetAvatarActorfromActorInfo))
    (:then
      (return (Abilities|Slam|GetSlamAttackCooldown _asdevice_overlord_spire)))
    (:CastFailed
      (return (|CallParentFunction)))))

(fn GetAbilityPreventTime
  (bind _asdevice_overlord_spire (Utilities|Casting|CastToDevice_OverlordSpire (Ability|GetAvatarActorfromActorInfo))
    (:then
      (return (Math|Random|RandomFloatinRange (Abilities|GetMinTimeBetweenAbilities _asdevice_overlord_spire) (Abilities|GetMaxTimeBetweenAbilities _asdevice_overlord_spire))))
    (:CastFailed
      (return (|CallParentFunction)))))

(fn CanActivateAbility
  (params (ActorInfo :type Gameplay%20Ability%20Actor%20Info%20Structure) (Handle :type Gameplay%20Ability%20Spec%20Handle%20Structure))
  (bind _asdevice_overlord_spire (Utilities|Casting|CastToDevice_OverlordSpire (Ability|GetAvatarActorfromActorInfo))
    (:then
      (if (Abilities|Slam|CanPerformSlamAttack _asdevice_overlord_spire)
        (bind (_relevanttags _returnvalue) (|CallParentFunction ActorInfo Handle))
        (return _relevanttags _returnvalue)
        (else
          _)))
    (:CastFailed
      _))
  (return 0 false))

(fn ShouldInstantDamageAffectPawn
  (params (Player :type Fort%20Pawn%20Object%20Reference))
  (if (GameplayTags|HasAnyMatchingGameplayTags Player (GameplayTags|MakeGameplayTagContainerfromTag "(TagName=\"Pawn.Athena.NPC.Deimos\")"))
    (return (> (ScalableFloat|GetValueatLevelwithContextParam (Ability|GetAvatarActorfromActorInfo) (LocalVariables|GetSF_DeimosDamage)) 0))
    (else
      (return (|CallParentFunction Player)))))

(fn SpawnGroundWaveActor
  (bind _spawnedactor (|CallParentFunction))
  (bind _ascrd_overlord_spire_device_ground_wave (Utilities|Casting|CastToCRD_OverlordSpireDevice_GroundWave _spawnedactor)
    (:then
      (Class|CRDOverlordSpireDeviceGroundWave|SetInstigatingAbility _ascrd_overlord_spire_device_ground_wave self :!class /CRD_OverlordSpire/GameplayActors/CRD_OverlordSpireDevice_GroundWave.CRD_OverlordSpireDevice_GroundWave_C)
      (Class|CRDOverlordSpireDeviceGroundWave|SetNormalizedSizeCurveValues _ascrd_overlord_spire_device_ground_wave (ScalableFloat|GetValueatLevelwithContextParam (Ability|GetAvatarActorfromActorInfo) (Variables|Default|GetSF_GroundWaveDuration)) (ScalableFloat|GetValueatLevelwithContextParam (Ability|GetAvatarActorfromActorInfo) (Variables|Default|GetSF_GroundWaveInitialRadius)) (ScalableFloat|GetValueatLevelwithContextParam (Ability|GetAvatarActorfromActorInfo) (Variables|Default|GetSF_GroundWaveMaxRadius))))
    (:CastFailed
      _))
  (return _spawnedactor))
