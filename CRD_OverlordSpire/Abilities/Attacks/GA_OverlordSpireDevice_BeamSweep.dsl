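; Overlord Spire device's beam sweep attack ability, overriding a parent ability with device-driven activation checks, cooldown and random prevent time.
; It notifies the device when the beam starts or stops, and ends if the spire is disabled.
; Damage varies by target type, filtering teammates and immune buildings, and scales with frame time.
; After long frames, it backtracks through skipped yaw with sphere traces, hitting each actor once per update.
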(event Ability|EventActivateAbility
  (|CallParentFunction)
  (bind _asdevice_overlord_spire (Utilities|Casting|CastToDevice_OverlordSpire (Variables|Internal|GetSpireActor)))
  (Abilities|Beam|NotifyBeamAttackEvent _asdevice_overlord_spire true)
  (Default|BindEventtoOnDisabled _asdevice_overlord_spire (EventDispatchers|CreateEvent :!event Custom|OnSpireDisabled)))

(event Ability|EventOnEndAbility
  (params (bWasCancelled :type Boolean))
  (|CallParentFunction bWasCancelled)
  (bind _asdevice_overlord_spire (Utilities|Casting|CastToDevice_OverlordSpire (Variables|Internal|GetSpireActor)))
  (Abilities|Beam|NotifyBeamAttackEvent _asdevice_overlord_spire false)
  (Default|UnbindEventfromOnDisabled _asdevice_overlord_spire (EventDispatchers|CreateEvent :!event Custom|OnSpireDisabled)))

(event Custom|OnSpireDisabled
  (Ability|EndAbility))

(fn CalculateDamageForCurrentActor
  (Utilities|Casting|CastToFortVehicleInterface (Variables|HitProcessing|GetCurrentTestActor)
    (:then
      (LocalVariables|SetDamageToApply (Variables|HitProcessing|GetSF_BeamDmg_Vehicle)))
    (:CastFailed
      (bind _asfort_player_pawn (Utilities|Casting|CastToFortPlayerPawn (Variables|HitProcessing|GetCurrentTestActor))
        (:then
          (if (not (or (Team|OnSameTeam (Variables|Internal|GetSpireActor) _asfort_player_pawn) (Vehicle|IsinVehicle _asfort_player_pawn)))
            (LocalVariables|SetDamageToApply (select (GameplayTags|HasMatchingGameplayTag _asfort_player_pawn "(TagName=\"Pawn.Athena.NPC.Creative.Guard\")") (Variables|Default|GetSF_BeamDmg_Guard) (Variables|HitProcessing|GetSF_BeamDmg_Player)))))
        (:CastFailed
          (bind _asbuilding_actor (Utilities|Casting|CastToBuildingActor (Variables|HitProcessing|GetCurrentTestActor))
            (:then
              (if (not (GameplayTags|HasMatchingGameplayTag _asbuilding_actor "(TagName=\"Gameplay.Status.Spire.Immune\")"))
                (LocalVariables|SetDamageToApply (select (GameplayTags|HasMatchingGameplayTag _asbuilding_actor "(TagName=\"Building.Type.Player\")") (Variables|HitProcessing|GetSF_BeamDmg_BuildPlayer) (Variables|HitProcessing|GetSF_BeamDmg_Building)))))
            (:CastFailed
              (bind _asfort_pawn (Utilities|Casting|CastToFortPawn (Variables|HitProcessing|GetCurrentTestActor)))
              (if (GameplayTags|HasMatchingGameplayTag _asfort_pawn "(TagName=\"Pawn.Athena.NPC.Wildlife\")")
                (LocalVariables|SetDamageToApply (Variables|Default|GetSF_BeamDmg_Wildlife))
                (elif (GameplayTags|HasMatchingGameplayTag _asfort_pawn "(TagName=\"Pawn.Athena.NPC.Deimos\")")
                  (LocalVariables|SetDamageToApply (Variables|Default|GetSF_BeamDmg_Creature))))))
          (Variables|HitProcessing|SetAppliedFireDmg (* (ScalableFloat|GetValueatLevelwithContextParam (Ability|GetAvatarActorfromActorInfo) (LocalVariables|GetDamageToApply)) (Variables|Default|GetTimeSinceLastFrame)))
          (return))))))

(fn GetAbilityCooldownTime
  (bind _asdevice_overlord_spire (Utilities|Casting|CastToDevice_OverlordSpire (Ability|GetAvatarActorfromActorInfo))
    (:then
      (return (Abilities|Beam|GetBeamAttackCooldown _asdevice_overlord_spire)))
    (:CastFailed
      (return (|CallParentFunction)))))

(fn GetAbilityPreventTime
  (bind _asdevice_overlord_spire (Utilities|Casting|CastToDevice_OverlordSpire (Ability|GetAvatarActorfromActorInfo))
    (:then
      (return (Math|Random|RandomFloatinRange (Abilities|GetMinTimeBetweenAbilities _asdevice_overlord_spire) (Abilities|GetMaxTimeBetweenAbilities _asdevice_overlord_spire))))
    (:CastFailed
      (return (|CallParentFunction)))))

(fn GetBeamHitUpdateTime
  (return (Math|Float|SelectFloat (Math|Float|SafeDivide (Variables|HitProcessing|GetBeamFireTime) (* (ScalableFloat|GetValueatLevelwithContextParam (Variables|Internal|GetSpireActor) (Variables|Default|GetSF_BeamYawEnd)) 2.0)) (|CallParentFunction) (> (* (ScalableFloat|GetValueatLevelwithContextParam (Variables|Internal|GetSpireActor) (Variables|Default|GetSF_BeamYawEnd)) 2.0) 0))))

(fn HandleBeamDamage
  (params (Hit :type Hit%20Result%20Structure%20%28by%20ref%29))
  (bind (_bblockinghit _binitialoverlap _time _distance _location _impactpoint _normal _impactnormal _physmat _hitactor) (Collision|BreakHitResult Hit))
  (if (not (Utilities|Array|ContainsItem (Variables|Default|GetActorsHitThisFrame) _hitactor))
    (Utilities|Array|AddUnique (Variables|Default|GetActorsHitThisFrame) _hitactor)
    (|CallParentFunction (Variables|GetHit))))

(fn HandleBeamTraceBacktrack
  (params (BeamStartTransform :type Vector) (BeamEndTransform :type Vector))
  (bind _outhits (Collision|MultiSphereTraceForObjects BeamStartTransform BeamEndTransform (ScalableFloat|GetValueatLevel (Variables|Internal|GetSF_BeamHitRadius)) (Variables|Internal|GetObject%20Types) false (Variables|Internal|GetAbilityActorsToIgnore) "None" true "(R=1.000000,G=0.000000,B=0.000000,A=1.000000)" "(R=0.000000,G=1.000000,B=0.000000,A=1.000000)" (Variables|HitProcessing|GetBeamFireTime)))
  (return _outhits))

(fn CanActivateAbility
  (params (ActorInfo :type Gameplay%20Ability%20Actor%20Info%20Structure) (Handle :type Gameplay%20Ability%20Spec%20Handle%20Structure))
  (bind _asdevice_overlord_spire (Utilities|Casting|CastToDevice_OverlordSpire (Ability|GetAvatarActorfromActorInfo))
    (:then
      (if (Abilities|Beam|CanPerformBeamAttack _asdevice_overlord_spire)
        (bind (_relevanttags _returnvalue) (|CallParentFunction ActorInfo Handle))
        (return _relevanttags _returnvalue)
        (else
          _)))
    (:CastFailed
      _))
  (return 0 false))

(fn ProcessBeamHitUpdate
  (bind _returnvalue (Utilities|Time|GetGameTimeinSeconds))
  (Utilities|FlowControl|Sequence
    (:then_0
      (bind _output_get (Variables|Default|SetTimeSinceLastFrame (- _returnvalue (Variables|ForCreative|GetLastFrametime))))
      (Variables|ForCreative|SetLastFrametime _returnvalue)
      (if (< _output_get (* (CallFunction|GetBeamHitUpdateTime) (LocalVariables|GetBacktrackThresholdMultiplier)))
        (|CallParentFunction)
        (bind _asdevice_overlord_spire (Utilities|Casting|CastToDevice_OverlordSpire (Ability|GetAvatarActorfromActorInfo)))
        (Variables|ForCreative|SetPreviousBeamSweepYaw (Variables|/SpireBoss/GameplayActors/BGA_Large_Spire.BGA_Large_Spire_C|GetBeamSweepCurrentYaw _asdevice_overlord_spire))
        (else
          (bind _asdevice_overlord_spire_2 (Utilities|Casting|CastToDevice_OverlordSpire (Ability|GetAvatarActorfromActorInfo)))
          (LocalVariables|SetBacktrackBeamYaw (Variables|/SpireBoss/GameplayActors/BGA_Large_Spire.BGA_Large_Spire_C|GetBeamSweepCurrentYaw _asdevice_overlord_spire_2))
          (while (> (LocalVariables|GetBacktrackBeamYaw) (Variables|ForCreative|GetPreviousBeamSweepYaw))
            (bind _result (Math|Float|DecrementFloat (LocalVariables|GetBacktrackBeamYaw))
              (:%20%20
                (bind (_outcurrentbeamtransform _outtransformbeamstoppos) (Abilities|Beam|ComputeBeamSweepTransforms _asdevice_overlord_spire_2 _result (Transformation|GetBoneTransform (Variables|/SpireBoss/GameplayActors/BGA_Large_Spire.BGA_Large_Spire_C|GetSkeletalMeshComponentBudgeted _asdevice_overlord_spire_2) "pillar_center" "RTS_World")))
                (bind _outcurrentbeamtransform_location (Math|Transform|BreakTransform _outcurrentbeamtransform))
                (bind _outtransformbeamstoppos_location (Math|Transform|BreakTransform _outtransformbeamstoppos))
                (bind _outhits (Default|HandleBeamTraceBacktrack _outcurrentbeamtransform_location _outtransformbeamstoppos_location))
                (for _array_element _outhits
                  (CallFunction|HandleBeamDamage _array_element)))))
          (Variables|ForCreative|SetPreviousBeamSweepYaw (Variables|/SpireBoss/GameplayActors/BGA_Large_Spire.BGA_Large_Spire_C|GetBeamSweepCurrentYaw _asdevice_overlord_spire_2)))))
    (:then_1
      (Utilities|Array|Clear (Variables|Default|GetActorsHitThisFrame)) ; Clear so that actors can be hit again on next beam trace
    )))
