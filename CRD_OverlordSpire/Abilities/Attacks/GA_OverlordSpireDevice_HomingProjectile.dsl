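; Device variant of the Overlord Spire homing projectile attack ability, driven by the spire device's settings.
; It gates activation, cooldown and prevent time through the device, notifies it of attack start and end, and ends when it is disabled.
; A scalable projectile cap is split across sources, remainder included; with too few projectiles, sources are shuffled and any target-facing bonus location is inserted.
; Optionally, ending waits until every source finishes spawning.
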(event Ability|EventActivateAbility
  (Variables|Default|SetMaxNumHomingProjectiles (Math|Float|Truncate (ScalableFloat|GetValueatLevelwithContextParam (Ability|GetAvatarActorfromActorInfo) (Variables|Default|GetSF_MaxNumHomingProjectiles))))
  (|CallParentFunction)
  (bind _asdevice_overlord_spire (Utilities|Casting|CastToDevice_OverlordSpire (Variables|Internal|GetSpireActor)))
  (Abilities|HomingProj|NotifyHomingProjectileAttackEvent _asdevice_overlord_spire true)
  (Default|BindEventtoOnDisabled _asdevice_overlord_spire (EventDispatchers|CreateEvent :!event Custom|OnSpireDisabled)))

(event Ability|EventOnEndAbility
  (params (bWasCancelled :type Boolean))
  (|CallParentFunction bWasCancelled)
  (bind _asdevice_overlord_spire (Utilities|Casting|CastToDevice_OverlordSpire (Variables|Internal|GetSpireActor)))
  (Abilities|HomingProj|NotifyHomingProjectileAttackEvent _asdevice_overlord_spire false)
  (Default|UnbindEventfromOnDisabled _asdevice_overlord_spire (EventDispatchers|CreateEvent :!event Custom|OnSpireDisabled)))

(event Custom|OnSpireDisabled
  (Ability|EndAbility))

(event EventSpawnProjectileSources
  (|CallParentFunction)
  (CallFunction|ResetInternalData))

(fn GetAbilityCooldownTime
  (bind _asdevice_overlord_spire (Utilities|Casting|CastToDevice_OverlordSpire (Ability|GetAvatarActorfromActorInfo))
    (:then
      (return (Abilities|HomingProj|GetHomingProjectileCooldown _asdevice_overlord_spire)))
    (:CastFailed
      (return (|CallParentFunction)))))

(fn GetAbilityPreventTime
  (bind _asdevice_overlord_spire (Utilities|Casting|CastToDevice_OverlordSpire (Ability|GetAvatarActorfromActorInfo))
    (:then
      (return (Math|Random|RandomFloatinRange (Abilities|GetMinTimeBetweenAbilities _asdevice_overlord_spire) (Abilities|GetMaxTimeBetweenAbilities _asdevice_overlord_spire))))
    (:CastFailed
      (return (|CallParentFunction)))))

(fn HaveAllProjectilesSpawned
  (return (>= (Variables|Default|GetNumSourcesFinishedSpawningProjectiles) (Utilities|Array|Length (Variables|Internal|GetHomingProjectileSourceArray)))))

(fn IsFinishedSpawningProjSources
  (return (or (|CallParentFunction) (>= (Variables|Default|GetProjectilesClaimed) (Variables|Default|GetMaxNumHomingProjectiles)))))

(fn CanActivateAbility
  (params (ActorInfo :type Gameplay%20Ability%20Actor%20Info%20Structure) (Handle :type Gameplay%20Ability%20Spec%20Handle%20Structure))
  (bind _asdevice_overlord_spire (Utilities|Casting|CastToDevice_OverlordSpire (Ability|GetAvatarActorfromActorInfo))
    (:then
      (if (Abilities|HomingProj|CanPerformHomingProjectileAttack _asdevice_overlord_spire)
        (bind (_relevanttags _returnvalue) (|CallParentFunction ActorInfo Handle))
        (return _relevanttags _returnvalue)
        (else
          _)))
    (:CastFailed
      _))
  (return 0 false))

(fn OnProjectileMontageEnd
  (bind _asdevice_overlord_spire (Utilities|Casting|CastToDevice_OverlordSpire (Variables|Internal|GetSpireActor))
    (:then
      (if (or (not (Variables|/CRD_OverlordSpire/Device_OverlordSpire.Device_OverlordSpire_C|GetbWaitForAllProjectilesToSpawnBeforeNextAttack _asdevice_overlord_spire)) (CallFunction|HaveAllProjectilesSpawned))
        (|CallParentFunction)))
    (:CastFailed
      (|CallParentFunction))))

(fn OnSourceFinishedSpawningProjectiles
  (Math|Integer|IncrementInt (Variables|Default|GetNumSourcesFinishedSpawningProjectiles)
    (:%20%20
      (bind _asdevice_overlord_spire (Utilities|Casting|CastToDevice_OverlordSpire (Variables|Internal|GetSpireActor)))
      (if (and (Variables|/CRD_OverlordSpire/Device_OverlordSpire.Device_OverlordSpire_C|GetbWaitForAllProjectilesToSpawnBeforeNextAttackRuntime _asdevice_overlord_spire) (CallFunction|HaveAllProjectilesSpawned))
        (Ability|EndAbility)))))

(fn PrepareProjSourceArrayForSpawn
  (if (< (Variables|Default|GetMaxNumHomingProjectiles) (Math|Integer|SelectInt (+ (Utilities|Array|Length (Variables|Internal|GetProjectileSourceArray)) 1) (Utilities|Array|Length (Variables|Internal|GetProjectileSourceArray)) (CallFunction|ShouldUseBonusLocation)))
    ; If less than 4 sources, ensure that the source facing the target always gets used
    (Utilities|Array|Shuffle (Variables|Internal|GetProjectileSourceArray))
    (if (CallFunction|ShouldUseBonusLocation)
      (Utilities|Array|Insert (Variables|Internal|GetProjectileSourceArray) (Variables|Internal|GetCurrentBestBonusLoc)))
    (else
      (|CallParentFunction))))

(fn ResetInternalData
  (Variables|Default|SetExtraPrjPerSourceClaimed 0)
  (Variables|Default|SetProjectilesClaimed 0)
  (Variables|Default|SetNumSourcesFinishedSpawningProjectiles 0))

(fn SpawnProjSource
  (params (Transform_Location :type Vector :default "0, 0, 0"))
  (bind _projectilesourcearray (Variables|Internal|GetProjectileSourceArray))
  (bind _returnvalue (Utilities|Array|Length _projectilesourcearray))
  (bind _bspawnextraprj (LocalVariables|GetbSpawnExtraPrj))
  (bind _baseprojectilespersource (LocalVariables|GetBaseProjectilesPerSource))
  (bind _returnvalue_2 (+ _baseprojectilespersource 1))
  (bind _returnvalue_3 (Math|Integer|SelectInt _returnvalue_2 _baseprojectilespersource _bspawnextraprj))
  (bind _projsource (|CallParentFunction Transform_Location))
  (bind _ascrd_overlord_spire_device_proj_source (Utilities|Casting|CastToCRD_OverlordSpireDevice_Proj_Source _projsource))
  (bind _output_get (LocalVariables|SetBaseProjectilesPerSource (/ (Variables|Default|GetMaxNumHomingProjectiles) _returnvalue)))
  (bind _output_get_2 (LocalVariables|SetbSpawnExtraPrj (> (- (- (Variables|Default|GetMaxNumHomingProjectiles) (* _output_get _returnvalue)) (Variables|Default|GetExtraPrjPerSourceClaimed)) 0)))
  (if _output_get_2
    (Math|Integer|IncrementInt (Variables|Default|GetExtraPrjPerSourceClaimed)
      (:%20%20
        _))
    (else
      _))
  (Class|CRDOverlordSpireDeviceProjSource|SetProjectileCountOverride _ascrd_overlord_spire_device_proj_source _returnvalue_3)
  (Variables|Default|SetProjectilesClaimed (+ (Variables|Default|GetProjectilesClaimed) _returnvalue_3)))
