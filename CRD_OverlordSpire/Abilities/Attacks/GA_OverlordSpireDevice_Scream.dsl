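; Gameplay ability for the Overlord Spire device's scream attack, gated on the spire's state and tags.
; Activation and end run parent logic, then notify the spire actor that the scream has started or stopped.
; CanActivateAbility requires the spire's CanPerformScreamAbility check and the Summon.Ready tag, without Summon.Cooldown or InitialCombat tags.
; It returns false when the avatar is not an Overlord Spire.
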(event Ability|EventActivateAbility
  (|CallParentFunction)
  (bind _asdevice_overlord_spire (Utilities|Casting|CastToDevice_OverlordSpire (Variables|Internal|GetSpireActor)))
  (Abilities|NotifyScreamAttackEvent _asdevice_overlord_spire true))

(event Ability|EventOnEndAbility
  (params (bWasCancelled :type Boolean))
  (|CallParentFunction bWasCancelled)
  (bind _asdevice_overlord_spire (Utilities|Casting|CastToDevice_OverlordSpire (Variables|Internal|GetSpireActor)))
  (Abilities|NotifyScreamAttackEvent _asdevice_overlord_spire false))

(fn CanActivateAbility
  (params (ActorInfo :type Gameplay%20Ability%20Actor%20Info%20Structure) (Handle :type Gameplay%20Ability%20Spec%20Handle%20Structure))
  (bind _asdevice_overlord_spire (Utilities|Casting|CastToDevice_OverlordSpire (Ability|GetAvatarActorfromActorInfo))
    (:then
      (return 0 (Math|Boolean|ANDBoolean (Abilities|CanPerformScreamAbility _asdevice_overlord_spire) (not (GameplayTags|HasMatchingGameplayTag _asdevice_overlord_spire "(TagName=\"Ability.SB.Summon.Cooldown\")")) (not (GameplayTags|HasMatchingGameplayTag _asdevice_overlord_spire "(TagName=\"Ability.SB.InitialCombat\")")) (GameplayTags|HasMatchingGameplayTag _asdevice_overlord_spire "(TagName=\"Ability.SB.Summon.Ready\")"))))
    (:CastFailed
      (return 0 false))))
